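# MF_MakePreciseVec2 is a library-exposed material function that makes a Vector2 input compile with the HLSL 'precise' keyword.
# A Custom node returns MakePrecise of the input on HLSL compilers above the ES3_1 feature level, and passes it through unchanged otherwise.
# Its description says this keeps values identical between depth pre-pass and base pass, avoiding depth mismatches seen as black flickering.
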MaterialFunction /Game/Athena/Environments/Landscape/Foliage/MF_MakePreciseVec2
graph {
 properties {bExposeToLibrary=true}
 nodes [
  MaterialExpressionFunctionInput_0 = FunctionInput() {
   description="Makes the graph piece connected to this input compile with the 'precise' keyword on DX11/DX12. This ensures that the value is calculated in a way so that it is the same on the depth pre pass and the base pass to avoid depth mismatching. (commonly seen as black flickering all over the object)"
   inputName=InVector
   inputType=FunctionInput_Vector2
  }
  n1 = Custom!(InValue=@MaterialExpressionFunctionInput_0) {
   code="##if COMPILER_HLSL &&&& FEATURE_LEVEL > FEATURE_LEVEL_ES3_1\r\nreturn MakePrecise(InValue);\r\n##else\r\nreturn InValue;\r\n##endif"
   description=MakePrecise2
   outputType=CMOT_Float2
   inputs [{inputName=InValue, input {expression=&self:MaterialExpressionFunctionInput_0}}]
  }
  FunctionOutput(None=@n1)
 ]
}
