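# Masked, two-sided landscape grass material for instanced and Nanite foliage, coloured by terrain colour match or a lake-grass switch, with gradient tip tint.
# World position offset combines noise height variation, panning wind that strengthens via a storm mask, and player bending.
# It also sinks instances as they fade and pushes grass from the camera; mobile disables offset.
# Curie landscape function wraps the attributes; biome-blend and safe-zone branches are unconnected.
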Material /Game/Athena/Environments/Landscape/Foliage/M_StWVolc_Grass_01
graph {
 properties {
  bTangentSpaceNormal=false
  bUseMaterialAttributes=true
  bUsedWithInstancedStaticMeshes=true
  bUsedWithNanite=true
  blendMode=BLEND_Masked
  ditheredLODTransition=true
  pixelDepthOffsetMode=PDOM_Legacy
  twoSided=true
  usedShadingModels=Default Lit
 }
 nodes [
  n1 = VectorParameter() {parameterName=GrassSample_Tint, defaultValue=<rgba 0.9 0.9 0.9 0>}
  n2 = VectorParameter() {parameterName=GrassSample_Tint_Arid, defaultValue=<rgba 0.9 0.9 0.9 0>}
  n3 = MaterialFunctionCall() {materialFunction=/Game/Athena/Environments/Landscape/MaterialFunctions/MF_Athena_BiomeMask}
  LinearInterpolate(A=@n1.RGB, B=@n2.RGB, Alpha=@n3.Swamp)
  n4 = MaterialFunctionCall() {materialFunction=/Game/Athena/Environments/Landscape/MaterialFunctions/MF_StormMask}
  n5 = LinearInterpolate(
   A=CollectionParameter() {collection=/Game/Athena/Environments/Landscape/MPC/MPC_Landscape, parameterName=Grass_WindSpeed}
   B=CollectionParameter() {collection=&MPC_Landscape, parameterName=Grass_WindSpeed_Storm}
   Alpha=@n4.Result
  )
  MaterialFunctionCall!(InVector=@n5) {
   desc="As of this note, this only compiles on PC.\r\nIt functions as a pass-through on other platforms.\r\nAdds 12(?) instructions"
   materialFunction=&/MF_MakePrecise
  }
  n6 = MaterialFunctionCall() {materialFunction=/Game/Athena/Environments/Landscape/MaterialFunctions/MF_Athena_TerrainColorMatch}
  n7 = Multiply(
   A=ComponentMask(None=CollectionParameter() {collection=&MPC_Landscape, parameterName=Arid_Grass_Tint}) {b=true, g=true, r=true}
   B=@n6.GrassColor
  )
  n8 = LinearInterpolate(
   A=@n6.Result
   B=Multiply(A=@n6.GrassNoColor, B=VectorParameter() {parameterName=Arid_SandGrass_Tint, defaultValue=<rgba 0.4678 0.314 0.1384 1>}.RGB)
   Alpha=@n3.Swamp
  )
  n9 = TextureSample(UVs=MaterialFunctionCall!(WorldPosition) {materialFunction=/Game/Athena/Environments/Landscape/MaterialFunctions/Utilities/MF_Athena_Terrain_Coords}.Result) {samplerType=SAMPLERTYPE_LinearColor, texture=/Game/Athena/Environments/Landscape/Textures/T_Athena_Terrain_Topo_Mask}
  MaterialFunctionCall!(BaseColor=StaticSwitchParameter(True=@n7, False=LinearInterpolate(A=@n8, B=@n7, Alpha=@n9.R)) {parameterName=ForceAridGrass}, Mask=@n3.Farm) {materialFunction=/Game/Athena/Environments/Landscape/MaterialFunctions/Jungle/MF_Athena_JunglePools}
  n10 = VectorParameter() {parameterName=GrassTipTint, defaultValue=<rgba 1.1 1.1 1.1 1>}
  n11 = Multiply(A=@n10.RGB, B=Constant() {r=1.2}) {constB=1.2}
  n12 = ScalarParameter() {defaultValue=0.5, parameterName=GradientMask_Min}
  n13 = ScalarParameter() {defaultValue=1, parameterName=GradientMask_Max}
  n14 = TextureSampleParameter2D() {
   parameterName=Gradient_Mask
   samplerSource=SSM_Wrap_WorldGroupSettings
   samplerType=SAMPLERTYPE_LinearColor
   texture=/Game/Athena/Prototype/Terrain/T_GrassGradient_01
  }
  n15 = Power(Base=Clamp(None=LinearInterpolate(A=@n12, B=@n13, Alpha=@n14.G)), Exp=ScalarParameter() {defaultValue=3, parameterName=Grass_Gradient_Pow})
  n16 = LinearInterpolate(
   A=StaticSwitchParameter(True=@n1.RGB, False=@n2.RGB) {parameterName=ForceAridGrass}
   B=ShadingPathSwitch(Default=@n10.RGB, Forward=@n11, Mobile=@n11)
   Alpha=@n15
  )
  n17 = Multiply(
   A=MaterialFunctionCall!(Water, GrassColor) {materialFunction=/Game/Athena/Environments/Landscape/MaterialFunctions/Standard/MF_Athena_LakeGrass_01}.Grass_Color
   B=@n16
  )
  n18 = Multiply(A=@n7, B=@n16)
  n19 = ScalarParameter() {defaultValue=0.02, parameterName=SpecMin}
  n20 = ScalarParameter() {defaultValue=0.2, parameterName=Spec}
  n21 = ScalarParameter() {defaultValue=0.6, parameterName=Rough}
  n22 = Constant() {r=1}
  n23 = TextureSampleParameter2D() {parameterName=MaskTexture, samplerSource=SSM_Wrap_WorldGroupSettings, texture=/Game/Environments/Foliage/Textures/T_WavyGrass_01}
  n24 = VertexNormalWS()
  n25 = TextureSample() {samplerSource=SSM_Wrap_WorldGroupSettings, samplerType=SAMPLERTYPE_Normal, texture=/Game/Environments/Foliage/Textures/T_WavyGrass_01_N}
  n26 = ComponentMask(None=TextureCoordinate() {coordinateIndex=1}) {g=true}
  n27 = LinearInterpolate(A=Transform(None=@n25.RGB), B=@n24, Alpha=Clamp(None=@n26))
  n28 = QualitySwitch(Default=@n27, Low=@n24, High=@n27, Medium=@n27)
  n29 = Constant3Vector() {constant=<rgba 0 0 1 0>}
  n30 = ComponentMask(None=WorldPosition().XYZ) {g=true, r=true}
  n31 = TextureSample(UVs=Divide(A=@n30, B=ScalarParameter() {defaultValue=1000, group=WPO, parameterName=HeightVarianceScale})) {
   samplerSource=SSM_Wrap_WorldGroupSettings
   samplerType=SAMPLERTYPE_LinearColor
   texture=/Game/Environments/Foliage/Textures/T_FieldGrass_01_M
  }
  n32 = Multiply(
   A=LinearInterpolate(
    A=ScalarParameter() {group=WPO, parameterName=StaticWPOLow}
    B=ScalarParameter() {defaultValue=30, group=WPO, parameterName=StaticWPOHigh}
    Alpha=@n31.R
   )
   B=PerInstanceFadeAmount()
  )
  n33 = CollectionParameter() {collection=&MPC_Landscape, parameterName=Ambient_WindMotion_Amount}
  n34 = ScalarParameter() {defaultValue=1200, group=WPO, parameterName=WindWaveScale}
  n35 = Time()
  n36 = TextureSample(UVs=Panner(Coordinate=Divide(A=@n30, B=@n34), Time=Multiply(A=@n35, B=@n5)) {speedY=1}) {samplerSource=SSM_Wrap_WorldGroupSettings, samplerType=SAMPLERTYPE_Normal, texture=/Game/Global/Textures/Default/Noise/T_SwirlNoise_N}
  n37 = LinearInterpolate(
   A=CollectionParameter() {collection=&MPC_Landscape, parameterName=Grass_WindPower}
   B=CollectionParameter() {collection=&MPC_Landscape, parameterName=Grass_WindPower_Storm}
   Alpha=@n4.Result
  )
  n38 = ComponentMask(None=Multiply(A=@n36.RGB, B=@n37)) {g=true, r=true}
  n39 = Constant() {r=1}
  n40 = Multiply(A=Multiply(A=Multiply(A=@n31.R, B=@n33), B=Constant() {r=300}), B=Normalize(VectorInput=AppendVector(A=@n38, B=@n39)))
  n41 = Constant3Vector() {constant=<rgba 1 1 0 0>}
  n42 = MaterialFunctionCall!(InVector=Add(A=Multiply(A=@n29, B=@n32), B=Multiply(A=@n40, B=@n41))) {
   desc="As of this note, this only compiles on PC.\r\nIt functions as a pass-through on other platforms.\r\nAdds 12(?) instructions"
   materialFunction=&/MF_MakePreciseVec3
  }
  n43 = CollectionParameter() {collection=&FortniteMaterialParameters, parameterName=PlayerPosition}
  n44 = MaterialFunctionCall!(
   PlayerPosition=@n43
   PlayerRadiusMax=CollectionParameter() {collection=&MPC_Landscape, parameterName=PlayerBend_Radius}
   BendStrength=CollectionParameter() {collection=&MPC_Landscape, parameterName=PlayerBend_Strength}
  ) {materialFunction=/Game/Athena/Environments/Landscape/MaterialFunctions/MF_AthenaFoliagePlayerInteract}
  n45 = Multiply(A=OneMinus(None=@n26), B=LinearInterpolate(A=@n42.Result, B=@n44.Rotation, Alpha=@n44.Rotation))
  n46 = Constant2Vector()
  n47 = Multiply(A=OneMinus(None=PerInstanceFadeAmount()), B=ScalarParameter() {defaultValue=-65, parameterName=SinkIntoGroundDist})
  n48 = CameraPositionWS()
  n49 = WorldPosition()
  n50 = ScalarParameter() {defaultValue=200, parameterName=camerafadeMaskRadius}
  n51 = ScalarParameter() {parameterName=camerafadeMaskHardness}
  n52 = WorldPosition()
  n53 = CameraPositionWS()
  n54 = Multiply(
   A=Multiply(A=Subtract(A=@n52.XYZ, B=@n53), B=Constant3Vector() {constant=<rgba 1 1 0 0>})
   B=ScalarParameter() {defaultValue=1, parameterName=CameraFadeMaskPush}
  )
  n55 = AppendVector(A=Constant2Vector(), B=ScalarParameter() {defaultValue=-200, parameterName=CameraFadeMaskSink})
  n56 = Add(
   A=Add(A=QualitySwitch(Default=@n45, High=@n45, Medium=@n45), B=AppendVector(A=@n46, B=@n47))
   B=Multiply(A=SphereMask(A=@n48, B=@n49.XYZ, Radius=@n50, Hardness=@n51), B=Add(A=@n54, B=@n55))
  )
  n57 = Constant()
  n58 = MaterialFunctionCall!(
   In=MakeMaterialAttributes!(
    BaseColor=StaticSwitchParameter(True=@n17, False=@n18) {parameterName=LakeGrass}
    Metallic
    Specular=LinearInterpolate(A=@n19, B=@n20, Alpha=@n15)
    Roughness=ShadingPathSwitch(Default=@n21, Mobile=@n22) {desc=Migrated from FeatureLevelSwitch}
    Anisotropy
    EmissiveColor
    Opacity
    OpacityMask=@n23.RGB
    Normal=ShadingPathSwitch(Default=@n28, Mobile=@n24) {desc=Migrated from FeatureLevelSwitch}
    Tangent
    WorldPositionOffset=ShadingPathSwitch(Default=@n56, Mobile=@n57) {desc=Migrated from FeatureLevelSwitch}
    SubsurfaceColor
    ClearCoat
    ClearCoatRoughness
    AmbientOcclusion
    Refraction
    CustomizedUVs_0
    CustomizedUVs_1
    CustomizedUVs_2
    CustomizedUVs_3
    CustomizedUVs_4
    CustomizedUVs_5
    CustomizedUVs_6
    CustomizedUVs_7
    PixelDepthOffset
    ShadingModel
    Displacement
   )
   UseHeightFalloff=StaticBool() {desc=Do Not UseHeightFalloff here}
   OnFireEmissiveOverride
   HeightFalloffRadius
   HeightFalloffHardness
   VariationMasksMobile
   In_Grass
   Is_Stylised_WE
  ) {desc=This should be near very last, materialFunction=/Game/Packages/Fortress_SharedMaterials/Base_Material_Functions/MF_Curie_Landscape}
  MaterialFunctionCall(MA In=@n58.Result) {materialFunction=&MF_CreativeMoveTool}
  MaterialFunctionCall(InVector=@n37) {
   desc="As of this note, this only compiles on PC.\r\nIt functions as a pass-through on other platforms.\r\nAdds 12(?) instructions"
   materialFunction=&MF_MakePrecise
  }
  n59 = ComponentMask(None=ObjectPositionWS()) {g=true, r=true}
  n60 = WorldPosition()
  n61 = Constant() {r=-1}
  n62 = CollectionParameter() {collection=&FortniteMaterialParameters, parameterName=AthenaSafeZoneCenter}
  n63 = Add(A=ComponentMask(None=Multiply(A=@n60.XYZ, B=@n61)) {g=true, r=true}, B=ComponentMask(None=@n62) {g=true, r=true})
  n64 = CollectionParameter() {collection=&FortniteMaterialParameters, parameterName=AthenaSafeZoneRadius}
  n65 = Saturate(None=Divide(A=Distance(A=@n59, B=@n63), B=@n64))
  n66 = Frac(None=@n65)
  n67 = Constant() {r=1500}
  n68 = Divide(A=Subtract(A=@n64, B=@n67) {constB=1000}, B=@n64)
  Max(A=Floor(None=@n65), B=Multiply(A=Subtract(A=@n66, B=@n68), B=Divide(A=@n64, B=@n67) {constB=1000}))
  ScalarParameter() {defaultValue=0.4, parameterName=Ambient Wind Motion Amount}
  ScalarParameter() {defaultValue=200, parameterName=PlayerBendRadius_1}
  ScalarParameter() {defaultValue=-50, parameterName=PlayerBendStrength}
 ]
}
